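# Sound cue for the thrust-start sound of the female Space Wanderer glider cosmetic.
# A Random node picks one of five equally weighted wave variations.
# It plays at reduced volume through the shared player glider thrust attenuation and the glider landing sound class.
# It references the shared glider thrust concurrency set and also carries per-owner, stop-oldest concurrency override values.
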SoundCue /BRCosmetics/Sounds/Gliders/SpaceWandererFemale/Glider_ThrustStart_SpaceWandererFemale
root=Random_0
nodes [
 {id=Random_0, children [WavePlayer_1,WavePlayer_3,WavePlayer_4,WavePlayer_6,WavePlayer_9], properties {weights [1,1,1,1,1]}}
 {id=WavePlayer_1, properties {soundWaveAssetPtr=&/Glider_SpaceWandererFemale_ThrustStart_01}}
 {id=WavePlayer_3, properties {soundWaveAssetPtr=&/Glider_SpaceWandererFemale_ThrustStart_02}}
 {id=WavePlayer_4, properties {soundWaveAssetPtr=&/Glider_SpaceWandererFemale_ThrustStart_03}}
 {id=WavePlayer_6, properties {soundWaveAssetPtr=&/Glider_SpaceWandererFemale_ThrustStart_05}}
 {id=WavePlayer_9, properties {soundWaveAssetPtr=&/Glider_SpaceWandererFemale_ThrustStart_06}}
]
properties {
 attenuationSettings=&Fort_Player_Glider_Thrust_Athena_Att
 maxDistance=12500
 soundClassObject=&Fort_Gliders_Land
 volumeMultiplier=0.5
 concurrencyOverrides {bLimitToOwner=true, maxCount=2, resolutionRule=StopOldest}
 concurrencySet [&Athena_Glider_Thrust_Conc]
}
